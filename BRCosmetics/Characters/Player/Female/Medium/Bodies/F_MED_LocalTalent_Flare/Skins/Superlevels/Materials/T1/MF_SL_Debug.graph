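# Debug material function for the Flare skin's Superlevels materials that displays its FX masks on the surface.
# ShowFX2map lerps between the FX1 and FX2 mask inputs, feeding base color and emissive while metallic, specular and roughness are zeroed.
# ShowFXMaps blends that debug view over the incoming material attributes, keeping their vertex attributes.
# Both switches are scalar parameters read from custom primitive data, so they can be set per primitive.
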MaterialFunction /BRCosmetics/Characters/Player/Female/Medium/Bodies/F_MED_LocalTalent_Flare/Skins/Superlevels/Materials/T1/MF_SL_Debug
graph {nodes [
  n1 = FunctionInput() {inputName=IN MA, inputType=FunctionInput_MaterialAttributes}
  n2 = LinearInterpolate(
   A=FunctionInput() {inputName=FX1Mask, sortPriority=10}
   B=FunctionInput() {inputName=FX2Mask, sortPriority=20}
   Alpha=ScalarParameter() {bUseCustomPrimitiveData=true, parameterName=ShowFX2map, primitiveDataIndex=2}
  )
  n3 = Constant3Vector() {constant=<rgba 0 0 1 0>}
  n4 = SetMaterialAttributes!(MaterialAttributes=@n1, BaseColor=@n2, EmissiveColor=@n2, Metallic=@n3.R, Specular=@n3.R, Roughness=@n3.R, ShadingModel=ShadingModel()) {attributeSetTypes [
    #1
    #2
    #3
    #4
    #5
    #6
   ]}
  FunctionOutput(None=BlendMaterialAttributes(A=@n1, B=@n4, Alpha=ScalarParameter() {bUseCustomPrimitiveData=true, parameterName=ShowFXMaps, primitiveDataIndex=1}) {vertexAttributeBlendType=UseA}) {outputName=Out MA}
 ]}
