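# Material function for the LocalTalent Flare superlevel character skin that layers audio-reactive glow effects onto the input material's emissive color.
# Radial gradients build the speaker and three star glows, and a stabilized screen-space dot texture adds a grille pattern.
# A screen waveform reuses Screech's MF_Wave with SDF masks, a sine curve and a secondary UV output for custom curve logic.
# Frontend-loop audio amplitude, or a sine test signal, supplies reactivity, alongside an EliminationActive input.
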MaterialFunction /BRCosmetics/Characters/Player/Female/Medium/Bodies/F_MED_LocalTalent_Flare/Skins/Superlevels/Materials/MF_LocalTalent_Flare_Techart_SL
graph {nodes [
  n1 = NamedRerouteUsage()
  n2 = NamedRerouteUsage()
  Add(A=@n1, B=@n2)
  ComponentMask(None=VectorParameter() {group=10_LocalTalent, parameterName=Wave UV Offset, sortPriority=64}.RGB) {g=true, r=true}
  n3 = GetMaterialAttributes(None=FunctionInput() {inputType=FunctionInput_MaterialAttributes}) {attributeGetTypes [#1]}
  n4 = NamedRerouteUsage()
  n5 = NamedRerouteDeclaration(None=@n3.EmissiveColor) {name=BaseEmission, nodeColor=<rgba 1 0.1641 0 1>}
  n6 = NamedRerouteUsage()
  n7 = NamedRerouteUsage()
  n8 = Multiply(A=ScalarParameter() {defaultValue=1, group=10_LocalTalent, parameterName=EnableSpeakerGlow, sortPriority=0}, B=NamedRerouteUsage()) {constA=1}
  n9 = Power(Base=NamedRerouteUsage()) {constExponent=6}
  n10 = LinearInterpolate(
   A=LinearInterpolate(A=LinearInterpolate(A=@n5, B=@n6, Alpha=@n7) {constAlpha=0}, B=Multiply(A=@n1, B=@n2), Alpha=Multiply(A=@n8, B=@n9))
   B=NamedRerouteUsage()
   Alpha=Multiply(A=NamedRerouteUsage(), B=ScalarParameter() {defaultValue=1, group=10_LocalTalent, parameterName=EnableWaveGlow})
  )
  n11 = NamedRerouteUsage()
  FunctionOutput(None=SetMaterialAttributes!(MaterialAttributes=@n3.MaterialAttributes, EmissiveColor=LinearInterpolate(A=@n4, B=@n10, Alpha=@n11)) {attributeSetTypes [#1]})
  Max(A=Add(A=@n5, B=NamedRerouteUsage()), B=@n2)
  n12 = TextureObjectParameter() {group=10_LocalTalent, parameterName=SpeakerMaskTexture, sortPriority=4, texture=/Game/Global/Textures/Default/Blanks/T_Black_srgb_EmptyAlpha}
  NamedRerouteDeclaration(None=TextureSample(Tex=@n12) {texture=/BRCosmetics/Weapons/FORT_Melee/Pickaxe_LocalTalent_Flare/Textures/T_LocalTalent_Flare_Pickaxe_Dots}.R) {name=TextureMapSpeakerMask, nodeColor=<rgba 0.5234 0 1 1>}
  n13 = NamedRerouteUsage()
  n14 = VectorParameter() {group=10_LocalTalentColors, parameterName=WaveGlowouterColor, sortPriority=64, defaultValue=<rgba 0.402 1 0.3763 1>}
  n15 = VectorParameter() {group=10_LocalTalentColors, parameterName=WaveGlowColor, sortPriority=64, defaultValue=<rgba 0.402 1 0.3763 1>}
  n16 = Constant2Vector() {g=0.05, r=-0.05}
  n17 = Constant2Vector() {g=0.5}
  n18 = OneMinus(None=MaterialFunctionCall(Input=NamedRerouteUsage(), Input Low=@n16.R, Input High=@n16.R, Target Low=@n17.R, Target High=@n17.R) {materialFunction=&RemapValueRange}.Result)
  n19 = Add(A=Multiply(A=NamedRerouteUsage(), B=Saturate(None=@n18)), B=NamedRerouteUsage())
  n20 = LinearInterpolate(Alpha=Power(Base=NamedRerouteUsage())) {constA=0.8}
  n21 = VectorParameter() {group=10_LocalTalentColors, parameterName=WaveCurveColor, sortPriority=64, defaultValue=<rgba 1 1 1 1>}
  n22 = LinearInterpolate(A=LinearInterpolate(A=@n14.RGB, B=@n15.RGB, Alpha=Multiply(A=@n19, B=@n20)), B=@n21.RGB, Alpha=NamedRerouteUsage())
  n23 = NamedRerouteUsage()
  NamedRerouteDeclaration(None=Multiply(A=LinearInterpolate(A=@n13, B=@n22, Alpha=@n23) {constAlpha=1})) {name=WaveOut, nodeColor=<rgba 0 1 0.2031 1>}
  n24 = Add(A=TextureCoordinate(), B=Constant2Vector() {g=-0.205, r=0.9728}) {constB=0.2}
  n25 = Constant() {r=2}
  NamedRerouteDeclaration(None=Rotator(
    Coordinate=Abs(None=@n24)
    Time=MaterialFunctionCall(Multiplier=@n25) {materialFunction=&Pi}.Result
   )) {name=WaveScreenUV, nodeColor=<rgba 0 1 0.5078 1>}
  n26 = ScalarParameter() {defaultValue=1, group=10_LocalTalent, parameterName=SpeakerRadius}
  n27 = Multiply(A=Constant() {r=0.09}, B=@n26)
  n28 = Constant() {r=5}
  n29 = Add(
   A=ComponentMask(None=VectorParameter() {group=10_LocalTalent, parameterName=SpeakerUVOffset, defaultValue=<rgba 0.388 0.095 0 0>}.RGB) {g=true, r=true}
   B=TextureCoordinate()
  ) {constB=0}
  NamedRerouteDeclaration(None=Multiply(
    A=MaterialFunctionCall(UVs=@n29, Radius=@n27, Density=@n28) {materialFunction=&RadialGradientExponential}.RadialGradientExponential
    B=NamedRerouteUsage()
   )) {name=SpeakerMask, nodeColor=<rgba 0 0.6953 1 1>}
  n30 = VectorParameter() {group=10_LocalTalentColors, parameterName=SpeakerOuterColor, sortPriority=34, defaultValue=<rgba 0.7682 0.9131 0.1274 1>}
  n31 = LinearInterpolate(
   A=@n30.RGB
   B=VectorParameter() {group=10_LocalTalentColors, parameterName=SpeakerColor, defaultValue=<rgba 0.7682 0.9131 0.1274 1>}.RGB
   Alpha=NamedRerouteUsage()
  )
  n32 = VectorParameter() {group=10_LocalTalentColors, parameterName=SpeakerInnerColor, sortPriority=30, defaultValue=<rgba 1 1 1 1>}
  n33 = MaterialFunctionCall(UVs=@n29, Radius=Multiply(A=Constant() {r=0.07}, B=@n26)) {materialFunction=&RadialGradientExponential}
  n34 = LinearInterpolate(
   A=@n31
   B=Add(A=@n31, B=@n32.RGB)
   Alpha=Multiply(A=NamedRerouteDeclaration(None=@n33.RadialGradientExponential) {name=SpeakerCirclePow, nodeColor=<rgba 1 0.6562 0 1>}, B=NamedRerouteUsage())
  )
  n35 = NamedRerouteUsage()
  n36 = Multiply(A=Constant() {r=0.025}, B=@n26)
  n37 = Constant() {r=10}
  n38 = MaterialFunctionCall!(
   Blend=LinearInterpolate(
    A=Multiply(A=@n34, B=@n35)
    Alpha=MaterialFunctionCall(UVs=@n29, Radius=@n36, Density=@n37) {materialFunction=&RadialGradientExponential}.RadialGradientExponential
   ) {constB=0}
   Base=NamedRerouteUsage()
  ) {materialFunction=/Engine/Functions/Engine_MaterialFunctions03/Blends/Blend_Screen}
  n39 = ScalarParameter() {defaultValue=1, group=10_LocalTalent, parameterName=SpeakerGlowBoost}
  NamedRerouteDeclaration(None=Multiply(A=Multiply(A=@n38.Result, B=@n39) {constB=1.25}, B=NamedRerouteUsage())) {name=SpeakerGlow, nodeColor=<rgba 1 0.9844 0 1>}
  n40 = ScalarParameter() {group=Audio Reactivity, parameterName=EliminationActive}
  n41 = Power(Base=NamedRerouteUsage())
  NamedRerouteDeclaration(None=LinearInterpolate(Alpha=Add(A=@n40, B=@n41)) {constA=-1}) {name=SpeakerCurve, nodeColor=<rgba 1 0 0.7031 1>}
  n42 = NamedRerouteUsage()
  NamedRerouteDeclaration(None=Saturate(None=Add(A=@n33.RadialGradientExponential, B=@n42) {constB=0})) {name=SpeakerCircle, nodeColor=<rgba 0 1 0.9766 1>}
  n43 = Constant() {r=0.001}
  n44 = Add(A=NamedRerouteUsage(), B=Power(Base=NamedRerouteUsage())) {constA=0.25}
  n45 = Multiply(A=@n44, B=OneMinus(None=NamedRerouteUsage()))
  NamedRerouteDeclaration(None=Multiply(
    A=MaterialFunctionCall!(Threshold=@n43, Gradient=@n45, Filter Width Override, Filter Width Fudge Factor) {materialFunction=/Game/UI/Materials/AAThreshold}.Result
    B=ScalarParameter() {group=10_LocalTalent, parameterName=EnableStarGlow, sortPriority=0}
   )) {name=StarGlowMask, nodeColor=<rgba 1 0.1172 0 1>}
  NamedRerouteDeclaration(None=@n29) {name=SpeakerUV, nodeColor=<rgba 0.7344 0 1 1>}
  n46 = Subtract(A=PixelDepth().R, B=ScalarParameter() {defaultValue=256, group=10_LocalTalent, parameterName=Wave Static Distance Fade Far, sortPriority=64}) {constB=128}
  n47 = ScalarParameter() {defaultValue=2, group=10_LocalTalent, parameterName=Wave Static Distance Fade Near, sortPriority=64}
  NamedRerouteDeclaration(None=Saturate(None=Divide(A=@n46, B=@n47) {constB=1024})) {name=DistanceFade, nodeColor=<rgba 0 1 0.8125 1>}
  n48 = OneMinus(None=NamedRerouteUsage())
  n49 = VectorParameter() {group=10_LocalTalent, parameterName=Star 1 UV}
  n50 = NamedRerouteUsage()
  n51 = Add(
   A=MaterialFunctionCall(
    CenterPosition=ComponentMask(None=@n49.RGB) {g=true, r=true}
    Radius=ScalarParameter() {defaultValue=0.05, group=10_LocalTalent, parameterName=Star 1 Radius}
   ) {materialFunction=&RadialGradientExponential}.RadialGradientExponential
   B=@n50
  )
  n52 = ScalarParameter() {defaultValue=4, group=10_LocalTalent, parameterName=Star 1 Boost}
  n53 = ComponentMask(None=VectorParameter() {group=10_LocalTalent, parameterName=Star 2 UV}.RGB) {g=true, r=true}
  n54 = ScalarParameter() {defaultValue=0.05, group=10_LocalTalent, parameterName=Star 2 Radius}
  n55 = NamedRerouteUsage()
  n56 = Multiply(
   A=Add(
    A=MaterialFunctionCall(CenterPosition=@n53, Radius=@n54) {materialFunction=&RadialGradientExponential}.RadialGradientExponential
    B=@n55
   )
   B=ScalarParameter() {defaultValue=4, group=10_LocalTalent, parameterName=Star 2 Boost}
  ) {constB=3.5}
  n57 = ComponentMask(None=VectorParameter() {group=10_LocalTalent, parameterName=Star 3 UV}.RGB) {g=true, r=true}
  n58 = ScalarParameter() {defaultValue=0.05, group=10_LocalTalent, parameterName=Star 3 Radius}
  n59 = NamedRerouteUsage()
  n60 = Multiply(
   A=Add(
    A=MaterialFunctionCall(CenterPosition=@n57, Radius=@n58) {materialFunction=&RadialGradientExponential}.RadialGradientExponential
    B=@n59
   )
   B=ScalarParameter() {defaultValue=4, group=10_LocalTalent, parameterName=Star 3 Boost}
  ) {constB=3.5}
  n61 = Add(A=Multiply(A=Multiply(A=@n51, B=@n52) {constB=3.5}, B=@n50), B=Add(A=Multiply(A=@n56, B=@n55), B=Multiply(A=@n60, B=@n59)))
  NamedRerouteDeclaration(None=Multiply(A=Multiply(A=@n48, B=@n61), B=@n44)) {name=StarGlow, nodeColor=<rgba 0 0.6719 1 1>}
  n62 = VectorParameter() {group=10_LocalTalentColors, parameterName=SpeakerOuterColor, sortPriority=34, defaultValue=<rgba 0.7682 0.9131 0.1274 1>}
  n63 = ScalarParameter() {defaultValue=1.59, group=10_LocalTalent, parameterName=DotsRotation, sortPriority=4}
  n64 = MaterialFunctionCall(UV Channel=Rotator(Coordinate=NamedRerouteUsage(), Time=@n63)) {materialFunction=&LinearGradient}
  n65 = MaterialFunctionCall!(Position, UVRadialScale) {materialFunction=/Game/Characters/Player/Common/MaterialFunctions/Superlevel/CharacterStabilizedScreenSpaceUVs}
  n66 = ScalarParameter() {defaultValue=100, group=10_LocalTalent, parameterName=Dots Scale, sortPriority=4}
  n67 = TextureObjectParameter() {
   desc=Grayscale pattern texture for effect over the grill.
   group=10_LocalTalent
   parameterName=SquareDots
   samplerType=SAMPLERTYPE_LinearGrayscale
   sortPriority=4
   texture=&T_LocalTalent_Flare_Pickaxe_Dots
  }
  n68 = TextureSample(
   UVs=Rotator(
    Coordinate=Multiply(A=@n65.StabilizedUVs, B=@n66)
    Time=ScalarParameter() {defaultValue=1.59, group=10_LocalTalent, parameterName=DotsRotation, sortPriority=4}
   )
   Tex=@n67
  ) {
   samplerType=SAMPLERTYPE_LinearGrayscale
   texture=&T_LocalTalent_Flare_Pickaxe_Dots
  }
  n69 = Multiply(A=Power(Base=@n68.R), B=NamedRerouteUsage())
  n70 = Multiply(
   A=LinearInterpolate(
    A=@n62.RGB
    B=VectorParameter() {group=10_LocalTalentColors, parameterName=SpeakerColor, defaultValue=<rgba 0.7682 0.9131 0.1274 1>}.RGB
    Alpha=Multiply(A=@n64.VGradient) {constB=1.75}
   )
   B=@n69
  )
  n71 = Add(A=@n70, B=Multiply(A=@n70, B=Power(Base=@n69) {constExponent=4}))
  n72 = ScalarParameter() {defaultValue=3, group=10_LocalTalent, parameterName=DotsBoost}
  NamedRerouteDeclaration(None=Multiply(A=Multiply(A=@n71, B=@n72), B=ScalarParameter() {group=10_LocalTalent, parameterName=EnableDots, sortPriority=0})) {name=Dots, nodeColor=<rgba 0 1 0.1094 1>}
  n73 = Add(
   A=NamedRerouteUsage()
   B=ComponentMask(None=VectorParameter() {group=10_LocalTalent, parameterName=ScreenUVOffset, defaultValue=<rgba -0.205 0 0 0>}.RGB) {g=true, r=true}
  )
  n74 = Constant2Vector()
  n75 = Constant2Vector() {g=1}
  n76 = Constant() {r=0.05}
  n77 = Multiply(A=@n76) {constB=0.9}
  n78 = Constant2Vector() {g=1}
  NamedRerouteDeclaration(None=MaterialFunctionCall(
    Input=MaterialFunctionCall!(UVs=@n73, PointA=@n74, PointB=@n75) {materialFunction=/FortUILibrary/MaterialFunctions/Shapes/SDFs/MF_SDF_LineSegment}.SDF
    Input Low=Add(A=@n76, B=@n77)
    Input High=Subtract(A=@n76, B=@n77)
    Target Low=@n78.R
    Target High=@n78.R
   ) {materialFunction=&RemapValueRange}.Result) {name=WaveFalloff, nodeColor=<rgba 0.125 1 0 1>}
  n79 = NamedRerouteUsage()
  n80 = Multiply(A=Add(A=ComponentMask(None=@n79) {r=true}, B=ScalarParameter() {defaultValue=-1, group=10_LocalTalent, parameterName=ScreenInnerGlowOffset}) {constB=-0.97}) {constB=12}
  NamedRerouteDeclaration(None=OneMinus(None=Saturate(None=@n80))) {name=ScreenGlow, nodeColor=<rgba 0 0.1328 1 1>}
  n81 = ConstantBiasScale(None=NamedRerouteUsage()) {bias=-0.5, scale=2}
  n82 = ScalarParameter() {defaultValue=32, group=10_LocalTalent, parameterName=WaveLength}
  n83 = Time()
  n84 = Constant() {r=-1.5}
  n85 = Add(A=Multiply(A=@n81, B=Multiply(A=@n82) {constB=0.5}), B=Frac(None=Multiply(A=@n83, B=@n84) {constB=0.1}))
  n86 = MaterialFunctionCall!(
   UV=@n81
   Wave Height=NamedRerouteUsage()
   Wave Length
   Wave Speed
   GlowMin
   GlowMax
   Outline Thickness=ScalarParameter() {defaultValue=0.01, group=10_LocalTalent, parameterName=Wave Outline Thickness}
   Custom Curve=Sine(None=FunctionOutput(None=@n85) {description=Output the UVs to input in your custom curve logic, outputName=Custom Curve UV, sortPriority=10})
  ) {materialFunction=/BRCosmetics/Characters/Player/Female/Medium/Bodies/F_MED_HangSpec_Screech/Skins/Material_Functions/MF_Wave}
  NamedRerouteDeclaration(None=LinearInterpolate(B=OneMinus(None=@n86.Outline), Alpha=NamedRerouteUsage())) {name=WaveOutline, nodeColor=<rgba 0.4766 0 1 1>}
  NamedRerouteDeclaration(None=Multiply(A=@n86.SDF Outline)) {name=WaveOutlineSDF, nodeColor=<rgba 0 1 0.0859 1>}
  n87 = MaterialFunctionCall!(InAmplitude, InSpectrum, CustomFakeSpectrum, CustomFakeAmplitude) {materialFunction=/BRCosmetics/Global/MaterialFunctions/AudioReactivity/MF_AudioReactivity_FrontendLoop}
  n88 = ScalarParameter() {defaultValue=1, group=10_LocalTalent, parameterName=EnableAudioReactivity, sliderMax=1}
  n89 = Sine(None=Time())
  n90 = Constant2Vector() {g=1, r=-1}
  NamedRerouteDeclaration(None=LinearInterpolate(
    A=Multiply(A=@n87.Amplitude, B=@n88)
    B=MaterialFunctionCall(Input=@n89, Input Low=@n90.R, Input High=@n90.R) {materialFunction=&RemapValue_0-1}.Result
    Alpha=ScalarParameter() {group=10_LocalTalent, parameterName=EnableTestAmplitude, sliderMax=1, sortPriority=0}
   )) {name=Amplitude, nodeColor=<rgba 0 1 0.1094 1>}
  n91 = ScalarParameter() {defaultValue=0.09, group=10_LocalTalent, parameterName=Wave Max Height, sortPriority=64}
  n92 = Multiply(A=@n91, B=Constant() {r=1.5})
  n93 = Power(Base=NamedRerouteUsage())
  NamedRerouteDeclaration(None=LinearInterpolate(B=LinearInterpolate(A=@n91, B=@n92, Alpha=@n93) {constB=0.1}, Alpha=NamedRerouteUsage())) {name=Wave Height, nodeColor=<rgba 1 0.7969 0 1>}
  n94 = NamedRerouteUsage()
  n95 = Constant2Vector() {r=1}
  n96 = MaterialFunctionCall!(
   UVs=Add(A=NamedRerouteUsage(), B=Constant2Vector() {g=-0.5, r=-0.41})
   Size=Constant2Vector() {g=0.03, r=0.11}
   CornerRoundness
   KeepAspectRatio
   PixelSize
  ) {materialFunction=/FortUILibrary/MaterialFunctions/Shapes/SDFs/MF_SDF_Box}
  n97 = Constant2Vector() {g=1}
  NamedRerouteDeclaration(None=Multiply(
    A=If(A=@n94, "A > B"=@n95.R, "A == B"=@n95.R, "A < B"=@n95.R) {constB=0.001}
    B=If(A=@n96.SDF, "A > B"=@n97.R, "A == B"=@n97.R, "A < B"=@n97.R)
   )) {name=ScreenMask, nodeColor=<rgba 0 0.3906 1 1>}
  OneMinus(None=@n33.RadialGradientExponential)
  n98 = TextureCoordinate()
  n99 = ComponentMask(None=VectorParameter() {group=10_LocalTalent, parameterName=DotsUVOffset}.RGB) {g=true, r=true}
  Rotator(
   Coordinate=Multiply(A=Add(A=@n98, B=@n99), B=ScalarParameter() {defaultValue=100, group=10_LocalTalent, parameterName=Dots Scale, sortPriority=4})
   Time=@n63
  )
 ]}
