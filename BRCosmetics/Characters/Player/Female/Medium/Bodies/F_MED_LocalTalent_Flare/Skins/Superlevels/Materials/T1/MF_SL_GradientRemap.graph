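# Material function in the Flare Superlevels skin materials that remaps a scalar gradient between StartPoint and EndPoint and builds a mask for that range.
# OUT Gradient saturates the offset from StartPoint divided by the signed range, which passes through an If on its absolute value.
# RangedMask multiplies a SmoothStep at the lower endpoint by an inverted one at the higher, with non-negative Smoothness setting both edge widths.
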MaterialFunction /BRCosmetics/Characters/Player/Female/Medium/Bodies/F_MED_LocalTalent_Flare/Skins/Superlevels/Materials/T1/MF_SL_GradientRemap
graph {nodes [
  n1 = FunctionInput() {bUsePreviewValueAsDefault=true, inputName=IN Gradient, inputType=FunctionInput_Scalar, previewValue=<1 0 0 1>}
  n2 = FunctionInput() {bUsePreviewValueAsDefault=true, inputName=StartPoint, inputType=FunctionInput_Scalar, sortPriority=1}
  n3 = Subtract(A=@n1, B=@n2)
  n4 = FunctionInput() {bUsePreviewValueAsDefault=true, inputName=EndPoint, inputType=FunctionInput_Scalar, sortPriority=2, previewValue=<1 0 0 1>}
  n5 = Subtract(A=@n4, B=@n2) {desc=Diff}
  n6 = If(A=Abs(None=@n5), "A > B"=@n5, "A < B"=Constant())
  FunctionOutput(None=Saturate(None=Divide(A=@n3, B=@n6))) {outputName=OUT Gradient}
  n7 = Min(A=@n2, B=@n4) {desc=MinV}
  n8 = FunctionInput() {bUsePreviewValueAsDefault=true, inputName=Smoothness, inputType=FunctionInput_Scalar, sortPriority=3, previewValue=<1 0 0 1>}
  n9 = Max(A=@n8) {constB=0, desc=S}
  n10 = Subtract(A=@n7, B=@n9)
  n11 = Add(A=@n7, B=@n9)
  n12 = Max(A=@n2, B=@n4) {constB=0, desc=MaxV}
  n13 = SmoothStep(Min=Subtract(A=@n12, B=@n9), Max=Add(A=@n12, B=@n9), Value=@n1) {desc=MaskEnd}
  FunctionOutput(None=Multiply(A=SmoothStep(Min=@n10, Max=@n11, Value=@n1) {desc=MaskStart}, B=OneMinus(None=@n13)) {desc=Mask}) {outputName=RangedMask, sortPriority=1}
 ]}
