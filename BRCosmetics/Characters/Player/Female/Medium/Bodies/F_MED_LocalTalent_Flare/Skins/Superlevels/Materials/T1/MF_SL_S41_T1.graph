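# Material function that builds the animated Superlevel S41 T1 pattern for the F_MED_LocalTalent_Flare skin, outputting primary, secondary and fur colors plus a pattern mask.
# A panning pattern texture is projected in pre-skinned local space, with noise distortion and a camera-based parallax offset that low quality skips.
# Three parameterized gradient masks and mask-shift controls shape the primary and secondary layers, with an FXMask input split into both.
# Primary color is a three-color gradient with adjustable end fills.
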MaterialFunction /BRCosmetics/Characters/Player/Female/Medium/Bodies/F_MED_LocalTalent_Flare/Skins/Superlevels/Materials/T1/MF_SL_S41_T1
graph {nodes [
  n1 = VectorParameter() {group=SL Pattern, parameterName=SL.Pattern_Secondary_ColorA, defaultValue=<rgba 0.015 0.015 0.015 1>}
  n2 = VectorParameter() {group=SL Pattern, parameterName=SL.Pattern_Secondary_ColorB, defaultValue=<rgba 0.0009 0.1714 0.1746 1>}
  FunctionOutput(None=LinearInterpolate(A=@n1.RGB, B=@n2.RGB, Alpha=NamedRerouteUsage())) {outputName=Pattern_Secondary_Color, sortPriority=20}
  FunctionOutput(None=VectorParameter() {group=SL Pattern, parameterName=SL.Fur_Color, defaultValue=<rgba 0.0089 0.1771 0.0937 1>}.RGB) {outputName=Fur_Pattern_Color, sortPriority=30}
  n3 = VectorParameter() {group=SL Pattern, parameterName=SL.Pattern_Primary_ColorA, defaultValue=<rgba 0.0356 0.9823 0.8963 1>}
  n4 = VectorParameter() {group=SL Pattern, parameterName=SL.Pattern_Primary_ColorMid, defaultValue=<rgba 0.1533 0.3309 0.8177 1>}
  n5 = VectorParameter() {group=SL Pattern, parameterName=SL.Pattern_Primary_ColorB, defaultValue=<rgba 0.9216 0.0222 0.2307 1>}
  n6 = NamedRerouteUsage()
  n7 = NamedRerouteUsage()
  n8 = MaterialFunctionCall(
   A=@n3.RGB
   B=@n4.RGB
   C=@n5.RGB
   Alpha=LinearInterpolate(A=OneMinus(None=@n6), B=@n6, Alpha=SmoothStep(Value=@n7) {constMax=0.55, constMin=0.45}) {constA=-1}
  ) {materialFunction=&Lerp_3Color}
  n9 = ScalarParameter() {group=SL Pattern, parameterName=SL.GradientMask01_ColorB_Fill, sliderMax=1, sortPriority=0}
  n10 = ScalarParameter() {group=SL Pattern, parameterName=SL.GradientMask02_ColorB_Fill, sliderMax=1, sortPriority=0}
  n11 = Max(
   A=SmoothStep(
    Min=ScalarParameter() {group=SL Pattern, parameterName=SL.GradientMask01_ColorA_Fill, sliderMax=1, sortPriority=0}
    Max=OneMinus(None=@n9)
    Value=NamedRerouteUsage()
   )
   B=SmoothStep(
    Min=ScalarParameter() {group=SL Pattern, parameterName=SL.GradientMask02_ColorA_Fill, sliderMax=1, sortPriority=0}
    Max=OneMinus(None=@n10)
    Value=NamedRerouteUsage()
   )
  )
  n12 = Saturate(None=Power(Base=@n11) {constExponent=3})
  n13 = Power(Base=OneMinus(None=@n11)) {constExponent=3}
  FunctionOutput(None=LinearInterpolate(A=LinearInterpolate(A=@n8.Result, B=@n5.RGB, Alpha=@n12), B=@n3.RGB, Alpha=Saturate(None=@n13))) {outputName=Pattern_Primary_Color, sortPriority=10}
  FunctionOutput(None=LinearInterpolate(A=NamedRerouteUsage(), B=NamedRerouteUsage(), Alpha=NamedRerouteUsage())) {outputName=Pattern_Mask, sortPriority=35}
  LinearInterpolate(A=@n1.RGB, B=@n2.RGB, Alpha=NamedRerouteUsage())
  n14 = NamedRerouteUsage()
  n15 = ScalarParameter() {defaultValue=1.5, group=SL Pattern, parameterName=SL.Pattern_Distortion_Tiling}
  n16 = MaterialFunctionCall!(
   TextureObject=TextureObject() {samplerType=SAMPLERTYPE_LinearGrayscale, texture=/Game/Global/Textures/Default/Noise/LowResBlurredNoise_NoMips}
   TextureSize=Multiply(A=Multiply(A=@n14, B=@n15) {constB=1.5}) {constB=32}
   WorldPosition=NamedRerouteUsage()
   FadingGradient
   In=Constant2Vector() {g=0.08, r=0.01}
   World Space Normal
   ProjectionTransitionContrast=Constant()
  ) {materialFunction=/BRCosmetics/Characters/Player/Male/Medium/Bodies/M_MED_FloralMane_Sand/Skins/SuperLevels/ParentMaterials/Preskinned_WP_Texture_ScalePan_XY}
  n17 = ComponentMask(None=@n16.XY Texture) {b=true}
  MaterialFunctionCall!(Linear=@n17) {materialFunction=/Engine/Functions/Engine_MaterialFunctions01/ImageAdjustment/LinearTosRGB}
  MaterialFunctionCall!(
   Position=Add(
    A=MaterialFunctionCall() {materialFunction=&ObjectPivotPoint}.Object Pivot Location
    B=Constant3Vector() {constant=<rgba 0 0 120 0>}
   )
   UVRadialScale
  ) {materialFunction=/Game/Characters/Player/Common/MaterialFunctions/Superlevel/CharacterStabilizedScreenSpaceUVs}
  n18 = ComponentMask(None=NamedRerouteUsage()) {g=true, r=true}
  Multiply(A=@n18) {constB=2}
  n19 = Multiply(A=TextureCoordinate(), B=@n18)
  n20 = NamedRerouteUsage()
  n21 = TextureSample(UVs=@n19, Tex=@n20) {samplerSource=SSM_Wrap_WorldGroupSettings}
  n22 = Add(A=@n19, B=Constant2Vector() {r=0.5})
  n23 = Constant2Vector() {g=0.025}
  n24 = Multiply(A=@n23) {constB=-1}
  Multiply(A=@n21.G, B=TextureSample(UVs=Panner(Coordinate=@n22, Speed=@n24), Tex=@n20) {samplerSource=SSM_Wrap_WorldGroupSettings}.G)
  n25 = LocalPosition() {localOrigin=InstancePreSkinning}
  NamedRerouteDeclaration(None=ShaderStageSwitch(PixelShader=VertexInterpolator(VS=@n25.XYZ).PS, VertexShader=@n25.XYZ)) {name=PreskinnedLocalPos, nodeColor=<rgba 1 0 0.0703 1>}
  n26 = PreSkinnedNormal()
  NamedRerouteDeclaration(None=ShaderStageSwitch(PixelShader=VertexInterpolator(VS=@n26).PS, VertexShader=@n26)) {name=PreskinnedLocalNormal, nodeColor=<rgba 1 0.9609 0 1>}
  n27 = VectorParameter() {
   group=SL Pattern
   parameterName=SL.GradientMask01_Range
   channelNames {a=Offset, b=Smoothness, g=EndPoint, r=StartPoint}
   defaultValue=<rgba 0 0.5 0 0>
  }
  n28 = MaterialFunctionCall!(
   IN Gradient=LinearInterpolate(A=NamedRerouteUsage(), B=NamedRerouteUsage(), Alpha=@n27.Offset)
   StartPoint=@n27.StartPoint
   EndPoint=@n27.StartPoint
   Smoothness=@n27.StartPoint
  ) {materialFunction=&/MF_SL_GradientRemap}
  n29 = VectorParameter() {
   group=SL Pattern
   parameterName=SL.GradientMask01_Transition
   channelNames {a=Hardness, b=Offset, g=Range, r=Amount}
   defaultValue=<rgba 0 1 0 1>
  }
  n30 = Multiply(
   A=MaterialFunctionCall!(Mask In=@n28.OUT Gradient, Amount=@n29.Amount, Range=@n29.Amount, Start Offset=@n29.Amount, Sharpness=@n29.Amount) {materialFunction=/Game/Global/MaterialFunctions/MF_MaskShifting}.Result
   B=@n28.RangedMask
  )
  NamedRerouteDeclaration(None=@n30) {name=Mask01Range, nodeColor=<rgba 0.9219 1 0 1>}
  NamedRerouteDeclaration(None=LinearInterpolate(
    A=VectorParameter() {group=SL Pattern, parameterName=SL.Pattern_Primary_Tiling, defaultValue=<rgba 4 4 4 1>}.RGB
    B=VectorParameter() {group=SL Pattern, parameterName=SL.Pattern_Secondary_Tiling, defaultValue=<rgba 4 4 4 1>}.RGB
    Alpha=NamedRerouteUsage()
   )) {name=GlobalTiling, nodeColor=<rgba 0 0.4375 1 1>}
  n31 = VectorParameter() {
   group=SL Pattern
   parameterName=SL.GradientMask02_Range
   channelNames {a=Offset, b=Smoothness, g=EndPoint, r=StartPoint}
   defaultValue=<rgba 0.5 1 0 0>
  }
  n32 = MaterialFunctionCall(
   IN Gradient=LinearInterpolate(A=NamedRerouteUsage(), B=NamedRerouteUsage(), Alpha=@n31.Offset)
   StartPoint=@n31.StartPoint
   EndPoint=@n31.StartPoint
   Smoothness=@n31.StartPoint
  ) {materialFunction=&MF_SL_GradientRemap}
  n33 = VectorParameter() {
   group=SL Pattern
   parameterName=SL.GradientMask02_Transition
   channelNames {a=Hardness, b=Offset, g=Range, r=Amount}
   defaultValue=<rgba 0 1 0 1>
  }
  n34 = Multiply(
   A=MaterialFunctionCall(Mask In=@n32.OUT Gradient, Amount=@n33.Amount, Range=@n33.Amount, Start Offset=@n33.Amount, Sharpness=@n33.Amount) {materialFunction=&MF_MaskShifting}.Result
   B=@n32.RangedMask
  )
  NamedRerouteDeclaration(None=@n34) {name=Mask02Rang, nodeColor=<rgba 0.9219 1 0 1>}
  n35 = TextureObject() {
   samplerType=SAMPLERTYPE_Masks
   texture=/BRCosmetics/Characters/Player/Female/Medium/Bodies/F_MED_LocalTalent_Flare/Skins/Superlevels/Textures/T_SL_S41_T1_Pattern
  }
  NamedRerouteDeclaration(None=@n35) {name=MaskTexture, nodeColor=<rgba 0 0.9297 1 1>}
  n36 = VectorParameter() {
   group=SL Pattern
   parameterName=SL.Pattern_Secondary_MaskTransition
   channelNames {a=Hardness, b=Offset, g=Range, r=Amount}
   defaultValue=<rgba 0 1 0 1>
  }
  NamedRerouteDeclaration(None=MaterialFunctionCall(Mask In=@n21.G, Amount=@n36.Amount, Range=@n36.Amount, Start Offset=@n36.Amount, Sharpness=@n36.Amount) {materialFunction=&MF_MaskShifting}.Result) {name=Fur_Secondary_Mask, nodeColor=<rgba 1 0.8672 0 1>}
  n37 = Multiply(A=NamedRerouteUsage(), B=ScalarParameter() {defaultValue=1, group=SL Pattern, parameterName=SL.GradientMask03_RangeTiling}) {constB=2}
  n38 = ScalarParameter() {group=SL Pattern, parameterName=SL.GradientMask03_RangeOffset}
  n39 = Multiply(A=Frac(None=Add(A=@n37, B=@n38) {constB=0})) {constB=2}
  n40 = VectorParameter() {
   group=SL Pattern
   parameterName=SL.GradientMask03_Range
   channelNames {a="", b=Smoothness, g=EndPoint, r=StartPoint}
   defaultValue=<rgba 0 1 0.1 0>
  }
  n41 = MaterialFunctionCall(IN Gradient=Abs(None=Subtract(A=@n39)), StartPoint=@n40.StartPoint, EndPoint=@n40.StartPoint, Smoothness=@n40.StartPoint) {materialFunction=&MF_SL_GradientRemap}
  n42 = VectorParameter() {
   group=SL Pattern
   parameterName=SL.GradientMask03_Transition
   channelNames {a=Hardness, b=Offset, g=Range, r=Amount}
   defaultValue=<rgba 0 1 0 0.4552>
  }
  n43 = Multiply(
   A=MaterialFunctionCall(Mask In=@n41.OUT Gradient, Amount=@n42.Amount, Range=@n42.Amount, Start Offset=@n42.Amount, Sharpness=@n42.Amount) {materialFunction=&MF_MaskShifting}.Result
   B=@n41.RangedMask
  )
  n44 = NamedRerouteUsage()
  NamedRerouteDeclaration(None=Saturate(None=Multiply(A=@n43, B=@n44))) {name=Pattern_Secondary_Mask, nodeColor=<rgba 0 1 0.9062 1>}
  n45 = VertexInterpolator(VS=LocalPosition() {localOrigin=InstancePreSkinning}.XYZ)
  n46 = PreSkinnedLocalBounds()
  n47 = ComponentMask(None=Divide(A=Subtract(A=@n45.PS, B=@n46.Min), B=@n46.Extents)) {b=true}
  n48 = Multiply(A=NamedRerouteUsage()) {constB=0.1}
  NamedRerouteDeclaration(None=Saturate(None=Add(A=@n47, B=@n48))) {name=PreSkinnedBoundingUV_B_Offseted, nodeColor=<rgba 0.2188 0 1 1>}
  n49 = NamedRerouteUsage()
  n50 = CameraPositionWS()
  n51 = Normalize(VectorInput=Subtract(A=@n49, B=TransformPosition(None=@n50) {transformSourceType=TRANSFORMPOSSOURCE_World}))
  n52 = ScalarParameter() {defaultValue=5, group=SL Pattern, parameterName=SL.Pattern_Parallax_Depth}
  n53 = QualitySwitch(Default=Add(A=@n49, B=Multiply(A=@n51, B=@n52) {constB=12}), Low=@n49)
  n54 = ScalarParameter() {defaultValue=8, group=SL Pattern, parameterName=SL.Pattern_Distort_Strength}
  n55 = MaterialFunctionCall(
   TextureObject=NamedRerouteUsage()
   TextureSize=Multiply(A=NamedRerouteUsage()) {constB=32}
   WorldPosition=Add(A=ShadingPathSwitch(Default=@n53, Deferred=@n53, Forward=@n49, Mobile=@n49), B=Multiply(A=@n17, B=@n54) {constB=0.04})
   In=Constant2Vector() {g=0.025}
   ProjectionTransitionContrast=ScalarParameter() {defaultValue=0.5, group=SL Pattern, parameterName=SL.Pattern_ProjectionContrast}
  ) {materialFunction=&Preskinned_WP_Texture_ScalePan_XY}
  NamedRerouteDeclaration(None=ComponentMask(None=@n55.XY Texture) {r=true}) {name=Primary_Mask, nodeColor=<rgba 1 0.8672 0 1>}
  NamedRerouteDeclaration(None=Saturate(None=Max(A=@n30, B=@n34))) {name=Pattern_Primary_Mask, nodeColor=<rgba 0 1 0.9063 1>}
  n56 = NamedRerouteUsage()
  n57 = VectorParameter() {
   group=SL Pattern
   parameterName=SL.Pattern_Primary_MaskTransition
   channelNames {a=Hardness, b=Offset, g=Range, r=Amount}
   defaultValue=<rgba 0 1 0 1>
  }
  NamedRerouteDeclaration(None=Saturate(None=MaterialFunctionCall(Mask In=@n56, Amount=@n57.Amount, Range=@n57.Amount, Start Offset=@n57.Amount, Sharpness=@n57.Amount) {materialFunction=&MF_MaskShifting}.Result)) {name=Tweaked_PrimaryMask, nodeColor=<rgba 0 1 0.2734 1>}
  NamedRerouteDeclaration(None=ComponentMask(None=@n55.XY Texture) {g=true}) {name=Secondary_Mask, nodeColor=<rgba 1 0.8672 0 1>}
  n58 = Normalize(VectorInput=VertexNormalWS())
  NamedRerouteDeclaration(None=Sign(None=NamedRerouteDeclaration(None=@n58) {name=VertexNormal, nodeColor=<rgba 0.25 0.25 1 1>})) {name=ProjectionSigns, nodeColor=<rgba 1 1 0 1>}
  n59 = NamedRerouteUsage()
  n60 = VectorParameter() {
   group=SL Pattern
   parameterName=SL.Pattern_Secondary_MaskTransition
   channelNames {a=Hardness, b=Offset, g=Range, r=Amount}
   defaultValue=<rgba 0 1 0 1>
  }
  NamedRerouteDeclaration(None=Saturate(None=MaterialFunctionCall(Mask In=@n59, Amount=@n60.Amount, Range=@n60.Amount, Start Offset=@n60.Amount, Sharpness=@n60.Amount) {materialFunction=&MF_MaskShifting}.Result)) {name=Tweaked_SecondaryMask, nodeColor=<rgba 0 1 0.2734 1>}
  NamedRerouteDeclaration(None=@n47) {name=PreSkinnedBoundingUV_B, nodeColor=<rgba 0.2188 0 1 1>}
  n61 = FunctionInput() {bUsePreviewValueAsDefault=true, inputName=FXMask, previewValue=<1 0 0 1>}
  NamedRerouteDeclaration(None=ComponentMask(None=@n61) {r=true}) {name=Primary, nodeColor=<rgba 1 0 0 1>}
  NamedRerouteDeclaration(None=ComponentMask(None=@n61) {b=true}) {name=Secondary, nodeColor=<rgba 0 0 1 1>}
  NamedRerouteUsage()
  Panner(Coordinate=@n19, Speed=@n23)
  Saturate(None=@n28.OUT Gradient)
 ]}
