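# Superlevel skin material for the LocalTalent Flare character, recolouring the character parent material through masked zones.
# Tech-art and FX2 mask channels select primary, secondary, tertiary, accent, hair and emissive-detail regions, each with colour or surface parameters.
# Static switches pick T1, T2 or T3 pattern functions for primary and secondary colours.
# Thin-film, metal LUT, Fresnel emissive and optional FN hair and LocalTalent reactivity functions add further effects.
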Material /BRCosmetics/Characters/Player/Female/Medium/Bodies/F_MED_LocalTalent_Flare/Skins/Superlevels/Materials/T1/M_SL_S41
graph {
 properties {
  bUseMaterialAttributes=true
  bUsedWithMeshDeformer=true
  bUsedWithMorphTargets=true
  bUsedWithSkeletalMesh=true
  pixelDepthOffsetMode=PDOM_Legacy
  shadingModel=MSM_FromMaterialExpression
  usedShadingModels=Default Lit | Subsurface | Clear Coat | Subsurface Profile | Cloth
 }
 nodes [
  n1 = NamedRerouteUsage()
  Add(A=@n1, B=Multiply(A=ScalarParameter() {group=SL 01 Global, parameterName=SL.Base_Roughness_Bias, sortPriority=1}, B=NamedRerouteUsage()))
  n2 = NamedRerouteUsage()
  Add(A=@n2, B=Multiply(A=ScalarParameter() {group=SL 01 Global, parameterName=SL.Base_Specular_Bias, sortPriority=1}, B=NamedRerouteUsage()))
  n3 = NamedRerouteUsage()
  Add(A=@n3, B=Multiply(A=ScalarParameter() {group=SL 01 Global, parameterName=SL.Base_Metallic_Bias, sortPriority=1}, B=NamedRerouteUsage()))
  n4 = NamedRerouteUsage()
  n5 = MaterialFunctionCall!(In=@n4) {materialFunction=/BRCosmetics/Characters/Player/Female/Medium/Bodies/F_MED_LocalTalent_Flare/Skins/Superlevels/Materials/MF_LocalTalent_Flare_Techart_SL}
  n6 = StaticSwitchParameter(True=@n5.Result, False=@n4) {group=10_LocalTalent, parameterName=Use.LocalTalentReactivity, sortPriority=-15}
  n7 = MaterialFunctionCall!(In=@n6, Mask Input=NamedRerouteUsage()) {materialFunction=/Game/Global/Materials/Character_Parent_2023/Hair_Parent_2023/MaterialFunctions/MF_FN_Hair_2023}
  n8 = MaterialFunctionCall() {materialFunction=&MF_CharacterParent_2023_PreFX}
  MaterialFunctionCall(
   Material=StaticSwitchParameter(True=BlendMaterialAttributes(A=@n6, B=@n7.Result, Alpha=@n7.Result), False=@n6) {group=SL 01 Global, parameterName=UseFNHair2_0}
   M=@n8.M
   Pre Skinned Local Position=@n8.M
   Distance Blend=@n8.M
   Fresnel=@n8.M
  ) {materialFunction=&MF_CharacterParent_2023_PostFX}
  NamedRerouteDeclaration(None=@n8.Result) {name=Material, nodeColor=<rgba 1 0 0.8672 1>}
  n9 = NamedRerouteUsage()
  n10 = ChannelMaskParameter(None=@n9) {group=SL 01 Global, maskChannel=Blue, parameterName=SL.SecondaryMaskChannel, sortPriority=-29, defaultValue=<rgba 0 0 1 0>}
  NamedRerouteDeclaration(None=@n10.R) {name=SecondaryMask, nodeColor=<rgba 0 0 1 1>}
  n11 = GetMaterialAttributes(None=NamedRerouteUsage()) {attributeGetTypes [
    #1
    #2
    #3
    #4
    #5
    #6
   ]}
  NamedRerouteDeclaration(None=@n11.Roughness) {name=Roughness, nodeColor=<rgba 0.7109 1 0 1>}
  n12 = NamedRerouteUsage()
  n13 = MaterialFunctionCall!(FXMask=@n12) {materialFunction=/BRCosmetics/Characters/Player/Female/Medium/Bodies/F_MED_LocalTalent_Flare/Skins/Superlevels/Materials/T3/MF_SL_S41_T3}
  n14 = MaterialFunctionCall!(FXMask=@n12) {materialFunction=/BRCosmetics/Characters/Player/Female/Medium/Bodies/F_MED_LocalTalent_Flare/Skins/Superlevels/Materials/T2/MF_SL_S41_T2}
  n15 = StaticBoolParameter() {group=SL 01 Global, parameterName=SL.Use_T3_Pattern, sortPriority=-40}
  n16 = StaticSwitch(True=@n13.Pattern_Secondary_Color, False=@n14.Pattern_Secondary_Color, Value=@n15)
  n17 = MaterialFunctionCall!(FXMask=@n12) {materialFunction=&/MF_SL_S41_T1}
  n18 = StaticBoolParameter() {group=SL 01 Global, parameterName=SL.Use_T2_Pattern, sortPriority=-40}
  n19 = NamedRerouteUsage()
  NamedRerouteDeclaration(None=Multiply(A=StaticSwitch(True=@n16, False=@n17.Pattern_Secondary_Color, Value=@n18), B=OneMinus(None=@n19))) {name=SecondaryColor, nodeColor=<rgba 0 0 1 1>}
  NamedRerouteDeclaration(None=ScalarParameter() {group=SL 01 Global, parameterName=SL.Tertiary_Metallic, sortPriority=2}) {name=TertiaryMetallic, nodeColor=<rgba 0 1 0 1>}
  NamedRerouteDeclaration(None=ScalarParameter() {group=SL 01 Global, parameterName=SL.Tertiary_Specular, sortPriority=2}) {name=TertiarySpecular, nodeColor=<rgba 0 1 0 1>}
  NamedRerouteDeclaration(None=ScalarParameter() {group=SL 01 Global, parameterName=SL.Tertiary_Roughness, sortPriority=2}) {name=TertiaryRoughness, nodeColor=<rgba 0 1 0 1>}
  NamedRerouteDeclaration(None=ScalarParameter() {group=SL 01 Global, parameterName=SL.Accent_Metallic, sortPriority=3}) {name=AccentMetallic, nodeColor=<rgba 0.25 1 0 1>}
  NamedRerouteDeclaration(None=ScalarParameter() {group=SL 01 Global, parameterName=SL.Accent_Specular, sortPriority=3}) {name=AccentSpecular, nodeColor=<rgba 0.25 1 0 1>}
  NamedRerouteDeclaration(None=ScalarParameter() {group=SL 01 Global, parameterName=SL.Accent_Roughness, sortPriority=3}) {name=AccentRoughness, nodeColor=<rgba 0.25 1 0 1>}
  NamedRerouteDeclaration(None=VectorParameter() {group=SL 01 Global, parameterName=SL.TertiaryColor, sortPriority=2}.RGB) {name=TertiaryColor, nodeColor=<rgba 0.2248 0.2274 1 1>}
  NamedRerouteDeclaration(None=VectorParameter() {group=SL 01 Global, parameterName=SL.AccentColor, sortPriority=3}.RGB) {name=AccentColor, nodeColor=<rgba 0.1854 0.7843 1 1>}
  NamedRerouteDeclaration(None=SmoothStep(
    Min=ScalarParameter() {defaultValue=0.5, group=SL 01 Global, parameterName=SL.TA_Mask_SmoothStepMin, sortPriority=100}
    Value=@n8.Tech Art Mask
   ) {constMin=0.5}) {name=TechArtMask, nodeColor=<rgba 1 0.4922 0 1>}
  NamedRerouteDeclaration(None=ScalarParameter() {group=SL 01 Global, parameterName=SL.Secondary_Metallic, sortPriority=1}) {name=SecondaryMetallic, nodeColor=<rgba 0 0 1 1>}
  NamedRerouteDeclaration(None=ScalarParameter() {group=SL 01 Global, parameterName=SL.Secondary_Specular, sortPriority=1}) {name=SecondarySpecular, nodeColor=<rgba 0 0 1 1>}
  NamedRerouteDeclaration(None=ScalarParameter() {group=SL 01 Global, parameterName=SL.Secondary_Roughness, sortPriority=1}) {name=SecondaryRoughness, nodeColor=<rgba 0 0 1 1>}
  NamedRerouteDeclaration(None=ChannelMaskParameter(None=@n9) {group=SL 01 Global, parameterName=SL.PrimaryMaskChannel, sortPriority=-32}.R) {name=PrimaryMask, nodeColor=<rgba 1 0 0 1>}
  NamedRerouteDeclaration(None=VectorParameter() {group=SL 01 Global, parameterName=SL.TertiaryColor2, sortPriority=2}.RGB) {name=TertiaryColor_2, nodeColor=<rgba 0.2248 0.2274 1 1>}
  n20 = LinearInterpolate(A=LinearInterpolate(A=@n1, B=NamedRerouteUsage(), Alpha=NamedRerouteUsage()), B=NamedRerouteUsage(), Alpha=NamedRerouteUsage())
  n21 = NamedRerouteUsage()
  n22 = NamedRerouteUsage()
  n23 = LinearInterpolate(
   A=LinearInterpolate(A=LinearInterpolate(A=@n20, B=@n21, Alpha=@n22), B=NamedRerouteUsage(), Alpha=NamedRerouteUsage())
   B=NamedRerouteUsage()
   Alpha=NamedRerouteUsage()
  )
  n24 = NamedRerouteUsage()
  n25 = LinearInterpolate(
   A=@n23
   B=Multiply(A=@n23, B=NamedRerouteUsage())
   Alpha=Multiply(A=ScalarParameter() {group=SL 01 Global, parameterName=SL.Roughness_GMap_Influence}, B=OneMinus(None=@n24))
  )
  n26 = Multiply(A=@n25, B=NamedRerouteUsage())
  n27 = NamedRerouteUsage()
  NamedRerouteDeclaration(None=LinearInterpolate(
    A=@n25
    B=LinearInterpolate(A=@n25, B=@n26, Alpha=@n27)
    Alpha=ScalarParameter() {group=SL 01 Global, parameterName=SL.Roughness_Pattern_Influence}
   )) {name=SL_Roughness, nodeColor=<rgba 1 0.338 0 1>}
  n28 = NamedRerouteUsage()
  n29 = NamedRerouteUsage()
  n30 = LinearInterpolate(
   A=LinearInterpolate(A=LinearInterpolate(A=@n2, B=@n28, Alpha=@n29), B=NamedRerouteUsage(), Alpha=NamedRerouteUsage())
   B=NamedRerouteUsage()
   Alpha=NamedRerouteUsage()
  )
  n31 = NamedRerouteUsage()
  n32 = NamedRerouteUsage()
  NamedRerouteDeclaration(None=LinearInterpolate(A=LinearInterpolate(A=@n30, B=@n31, Alpha=@n32), B=NamedRerouteUsage(), Alpha=NamedRerouteUsage())) {name=SL_Specular, nodeColor=<rgba 0.9918 1 0.1851 1>}
  n33 = MaterialFunctionCall(Float3=@n8.M) {desc="R = AO\r\nG = GMap\r\nB = Skin", materialFunction=&BreakOutFloat3Components}
  NamedRerouteDeclaration(None=@n33.B) {name=SkinMask, nodeColor=<rgba 0.2188 1 0 1>}
  NamedRerouteDeclaration(None=StaticSwitch(True=StaticSwitch(True=@n13.Pattern_Mask, False=@n14.Pattern_Mask, Value=@n15), False=@n17.Pattern_Mask, Value=@n18)) {name=PatternMask, nodeColor=<rgba 1 0 0.8672 1>}
  n34 = ScalarParameter() {defaultValue=0.5, group=SL 01 Global, parameterName=SL.TA_Mask_SmoothStepMin, sortPriority=100}
  n35 = TextureSampleParameter2D() {
   group=SL 01 Global
   parameterName=SL.FX2Mask
   samplerSource=SSM_Wrap_WorldGroupSettings
   samplerType=SAMPLERTYPE_Masks
   sortPriority=-30
   texture=&T_BlackMask
  }
  n36 = StaticSwitchParameter(
   True=NamedRerouteDeclaration(None=SmoothStep(Min=@n34, Value=@n35.RGB) {constMin=0.5}) {name=FX2Mask, nodeColor=<rgba 1 0.4922 0 1>}
   False=NamedRerouteUsage()
  ) {defaultValue=true, group=SL 01 Global, parameterName=SL.UseFX2Mask}
  NamedRerouteDeclaration(None=SmoothStep(
    Min=ScalarParameter() {group=SL 01 Global, parameterName=EmissiveDetail_SmoothStep_MIN}
    Max=ScalarParameter() {defaultValue=1, group=SL 01 Global, parameterName=EmissiveDetail_SmoothStep_MAX}
    Value=ChannelMaskParameter(None=@n36) {group=SL 01 Global, parameterName=SL.EmissiveDetailMaskChannel, sortPriority=-27}.R
   )) {name=EmissiveDetailMask, nodeColor=<rgba 1 0 0.25 1>}
  NamedRerouteDeclaration(None=@n8.Fresnel) {name=Fresnel_Input, nodeColor=<rgba 0.4297 0 1 1>}
  n37 = ChannelMaskParameter(None=@n36) {group=SL 01 Global, maskChannel=Blue, parameterName=SL.HairMaskChannel, sortPriority=-27, defaultValue=<rgba 0 0 1 0>}
  NamedRerouteDeclaration(None=@n37.R) {name=HairMask, nodeColor=<rgba 0 0.25 1 1>}
  n38 = VectorParameter() {
   group=SL 01 Global
   parameterName=SL.Hair_GradientMask_Remap
   channelNames {a="", b=Smoothness, g=EndPoint, r=StartPoint}
   defaultValue=<rgba 0 0.5 0.1 1>
  }
  n39 = MaterialFunctionCall!(
   IN Gradient=StaticSwitchParameter(True=NamedRerouteUsage(), False=NamedRerouteUsage()) {group=SL 01 Global, parameterName=SL.Hair_Use_SphereGradient}
   StartPoint=@n38.StartPoint
   EndPoint=@n38.StartPoint
   Smoothness=@n38.StartPoint
  ) {materialFunction=&/MF_SL_GradientRemap}
  n40 = VectorParameter() {
   group=SL 01 Global
   parameterName=SL.Hair_GradientMask_Transition
   channelNames {a=Hardness, b=Offset, g=Range, r=Amount}
   defaultValue=<rgba 0 1 0 1>
  }
  n41 = Multiply(
   A=MaterialFunctionCall!(Mask In=@n39.OUT Gradient, Amount=@n40.Amount, Range=@n40.Amount, Start Offset=@n40.Amount, Sharpness=@n40.Amount) {materialFunction=/Game/Global/MaterialFunctions/MF_MaskShifting}.Result
   B=@n39.RangedMask
  )
  NamedRerouteDeclaration(None=LinearInterpolate(
    A=VectorParameter() {group=SL 01 Global, parameterName=SL.HairColor01}.RGB
    B=VectorParameter() {group=SL 01 Global, parameterName=SL.HairColor02}.RGB
    Alpha=Saturate(None=@n41)
   )) {name=HairColor, nodeColor=<rgba 0 0.25 1 1>}
  NamedRerouteDeclaration(None=ScalarParameter() {group=SL 01 Global, parameterName=SL.Hair_Metallic}) {name=Hair_Metallic, nodeColor=<rgba 0 0.25 1 1>}
  NamedRerouteDeclaration(None=ScalarParameter() {group=SL 01 Global, parameterName=SL.Hair_Specular}) {name=Hair_Specular, nodeColor=<rgba 0 0.25 1 1>}
  NamedRerouteDeclaration(None=VectorParameter() {group=SL 01 Global, parameterName=SL.EmissiveDetailColor, sortPriority=3}.RGB) {name=EmissiveDetailColor, nodeColor=<rgba 1 0 0.25 1>}
  NamedRerouteDeclaration(None=@n11.Emissive Color) {name=Emissive, nodeColor=<rgba 1 0.4435 0.2797 1>}
  NamedRerouteDeclaration(None=@n11.Normal) {name=Normal, nodeColor=<rgba 1 0.1712 0.7455 1>}
  NamedRerouteDeclaration(None=ScalarParameter() {group=SL 01 Global, parameterName=SL.Hair_Roughness}) {name=Hair_Roughness, nodeColor=<rgba 0 0.25 1 1>}
  n42 = NamedRerouteUsage()
  n43 = ScalarParameter() {defaultValue=1, group=SL 01 Global, parameterName=SL.Tertiary_ThinFilmScale, sortPriority=5}
  n44 = Power(
   Base=NamedRerouteUsage()
   Exp=ScalarParameter() {defaultValue=0.5, group=SL 01 Global, parameterName=SL.Tertiary_ThinFilmWarp, sortPriority=4}
  )
  n45 = TextureSampleParameter2D(UVs=Add(A=Multiply(A=@n43, B=@n44), B=ScalarParameter() {group=SL 01 Global, parameterName=SL.Tertiary_ThinFilmOffset})) {
   group=SL 01 Global
   parameterName=SL.Tertiary_ThinFilmTexture
   samplerSource=SSM_Wrap_WorldGroupSettings
   sortPriority=1
   texture=/Game/Global/Textures/Default/Thinfilm/T_ThinFilm_Spectrum_COLOR
  }
  n46 = ScalarParameter() {defaultValue=0.5, group=SL 01 Global, parameterName=SL.Global_ThinFilmIntensity, sortPriority=2}
  n47 = ScalarParameter() {defaultValue=1, group=SL 01 Global, parameterName=SL.Tertiary_ThinFilmScale, sortPriority=5}
  n48 = NamedRerouteUsage()
  n49 = Power(Base=@n48, Exp=ScalarParameter() {defaultValue=0.5, group=SL 01 Global, parameterName=SL.Tertiary_ThinFilmWarp, sortPriority=4})
  n50 = Multiply(
   A=Multiply(A=Multiply(A=@n45.RGB, B=@n46), B=Multiply(A=@n47, B=@n49))
   B=ScalarParameter() {defaultValue=0.5, group=SL 01 Global, parameterName=SL.Tertiary_ThinFilmIntensity, sortPriority=2}
  )
  n51 = Power(Base=@n48, Exp=ScalarParameter() {defaultValue=2, group=SL 01 Global, parameterName=SL.Global_ThinFilmExponent, sortPriority=3})
  n52 = NamedRerouteUsage()
  n53 = LinearInterpolate(
   A=@n51
   B=Multiply(A=LinearInterpolate(Alpha=@n52) {constA=1, constB=0.01}, B=@n51)
   Alpha=ScalarParameter() {group=SL 01 Global, parameterName=SL.Global_RoughnessInfluence}
  )
  n54 = NamedRerouteUsage()
  n55 = NamedRerouteUsage()
  n56 = LinearInterpolate(
   A=@n42
   B=LinearInterpolate(A=@n42, B=Add(A=@n42, B=@n50), Alpha=Saturate(None=@n53))
   Alpha=StaticSwitchParameter(True=Max(A=@n54, B=@n55), False=@n54) {group=SL 01 Global, parameterName=SL.UseSharedThinfilmOnAccent, sortPriority=-30}
  )
  n57 = CameraVectorWS()
  n58 = PixelNormalWS()
  n59 = CurveAtlasRowParameter(CurveTime=Clamp(None=DotProduct(A=@n57, B=@n58))) {
   atlas=/Game/Global/Textures/Default/Curve/FN_Character_Atlas
   curve=/Game/Characters/Player/Female/Medium/Bodies/F_MED_Ms_Alpine/Textures/AluminiumreflectanceCurve
   group=SL 01 Global
   parameterName=SL.Metal LUT Curve
  }
  n60 = Multiply(A=@n56, B=ChannelMaskParameter(None=@n59) {group=SL 01 Global, parameterName=SL.LUTChannel}.R)
  n61 = Saturate(None=ScalarParameter() {defaultValue=1, group=SL 01 Global, parameterName=SL.MetalLutIntensity})
  NamedRerouteDeclaration(None=LinearInterpolate(A=@n56, B=LinearInterpolate(A=@n56, B=@n60, Alpha=@n61), Alpha=NamedRerouteUsage())) {name=SL_BaseColor_Metal, nodeColor=<rgba 1 0 0.0938 1>}
  n62 = PreSkinnedLocalBounds()
  n63 = Subtract(A=VertexInterpolator(VS=LocalPosition() {localOrigin=InstancePreSkinning}.XYZ).PS, B=@n62.Min)
  NamedRerouteDeclaration(None=ComponentMask(None=Divide(A=@n63, B=@n62.Extents)) {b=true}) {name=PreSkinnedBoundingUV_B, nodeColor=<rgba 0.2188 0 1 1>}
  NamedRerouteDeclaration(None=ScalarParameter() {group=SL 01 Global, parameterName=SL.EmissiveDetailStrength, sortPriority=3}) {name=EmissiveDetailStrength, nodeColor=<rgba 1 0 0.25 1>}
  n64 = LinearInterpolate(
   A=NamedRerouteUsage()
   B=Multiply(A=NamedRerouteUsage(), B=ScalarParameter() {defaultValue=0.6, group=SL 01 Global, parameterName=SL.Pattern_Primary_Emissive_Brightness})
   Alpha=NamedRerouteUsage()
  )
  n65 = Multiply(A=NamedRerouteUsage(), B=ScalarParameter() {defaultValue=0.6, group=SL 01 Global, parameterName=SL.Pattern_Secondary_Emissive_Brightness})
  n66 = NamedRerouteUsage()
  n67 = LinearInterpolate(A=NamedRerouteUsage(), B=NamedRerouteUsage(), Alpha=NamedRerouteUsage())
  n68 = NamedRerouteUsage()
  NamedRerouteDeclaration(None=LinearInterpolate(A=LinearInterpolate(A=@n64, B=@n65, Alpha=@n66), B=Multiply(A=@n67, B=@n68), Alpha=NamedRerouteUsage())) {name=SL_Emissive, nodeColor=<rgba 0 0.7422 1 1>}
  n69 = ScalarParameter() {defaultValue=1, group=SL 01 Global, parameterName=SL.EmissiveFresnel_Brightness, sliderMin=4, sortPriority=6}
  n70 = ScalarParameter() {defaultValue=1.5, group=SL 01 Global, parameterName=SL.EmissiveFresnel_Exponent, sortPriority=7}
  n71 = PixelNormalWS()
  NamedRerouteDeclaration(None=Multiply(
    A=MaterialFunctionCall(Brightness=@n69, Exponent=@n70, Vertex / Pixel Normal=@n71) {materialFunction=&MF_Fresnel_FX}.Fresnel Inverted
    B=NamedRerouteUsage()
   )) {name=EmissiveFresnel, nodeColor=<rgba 1 0 0.2812 1>}
  NamedRerouteDeclaration(None=VectorParameter() {group=SL 01 Global, parameterName=SL.EmissiveFresnelColor, sortPriority=3}.RGB) {name=EmissiveFresnelColor, nodeColor=<rgba 1 0 0.25 1>}
  NamedRerouteDeclaration(None=VectorParameter() {group=SL 01 Global, parameterName=SL.EmissiveDetailBaseColor, sortPriority=3}.RGB) {name=EmissiveDetaiBaseColor, nodeColor=<rgba 1 0 0.25 1>}
  NamedRerouteDeclaration(None=@n33.R) {name=AO, nodeColor=<rgba 1 0.1172 0 1>}
  NamedRerouteDeclaration(None=@n33.G) {name=GMap, nodeColor=<rgba 0 0.7309 1 1>}
  n72 = VertexInterpolator(VS=LocalPosition() {localOrigin=InstancePreSkinning}.XYZ)
  n73 = VectorParameter() {group=SL 01 Global, parameterName=SL.Hair_Gradient_SpherePos, channelNames {a=Radius, b="", g="", r=""}, defaultValue=<rgba 0 0 170 1>}
  NamedRerouteDeclaration(None=OneMinus(None=SphereMask(A=@n72.PS, B=@n73.RGB, Radius=@n73.RGB) {attenuationRadius=10, hardnessPercent=0})) {name=HairGradientMap, nodeColor=<rgba 0.2188 0 1 1>}
  NamedRerouteDeclaration(None=DotProduct(A=@n9, B=Constant() {r=1})) {name=FlatTechArtMask, nodeColor=<rgba 1 0 0.1641 1>}
  n74 = ChannelMaskParameter(None=@n9) {group=SL 01 Global, maskChannel=Green, parameterName=SL.TertiaryMaskChannel, sortPriority=-28, defaultValue=<rgba 0 1 0 0>}
  NamedRerouteDeclaration(None=SmoothStep(
    Min=ScalarParameter() {group=SL 01 Global, parameterName=Tertiary_SmoothStep_MIN}
    Max=ScalarParameter() {defaultValue=1, group=SL 01 Global, parameterName=Tertiary_SmoothStep_MAX}
    Value=@n74.R
   )) {name=TertiaryMask, nodeColor=<rgba 0 1 0 1>}
  NamedRerouteDeclaration(None=OneMinus(None=NamedRerouteUsage())) {name=SL_BaseMask, nodeColor=<rgba 0.3125 0 1 1>}
  n75 = LinearInterpolate(A=NamedRerouteUsage(), B=NamedRerouteUsage(), Alpha=NamedRerouteUsage())
  n76 = NamedRerouteUsage()
  n77 = SetMaterialAttributes!(
   MaterialAttributes=NamedRerouteUsage()
   Base Color=LinearInterpolate(A=@n75, B=Multiply(A=@n75, B=@n76), Alpha=ScalarParameter() {group=SL 01 Global, parameterName=SL.BaseColor_AO_Influence})
   Metallic=NamedRerouteUsage()
   Specular=NamedRerouteUsage()
   Roughness=NamedRerouteUsage()
   Emissive Color=NamedRerouteUsage()
  ) {attributeSetTypes [
    #1
    #2
    #3
    #4
    #5
   ]}
  NamedRerouteDeclaration(None=@n77) {name=SuperLevelsMaterial, nodeColor=<rgba 0 0.3906 1 1>}
  n78 = Multiply(
   A=NamedRerouteUsage()
   B=VectorParameter() {group=SL 01 Global, parameterName=SL.BaseColorTint, sortPriority=2, defaultValue=<rgba 1 1 1 1>}.RGB
  )
  n79 = VectorParameter() {group=SL 01 Global, parameterName=SL.Primary_BaseColorAdd, sortPriority=2, defaultValue=<rgba 0 0 0 1>}
  n80 = Add(
   A=@n79.RGB
   B=Multiply(A=NamedRerouteUsage(), B=ScalarParameter() {defaultValue=0.4, group=SL 01 Global, parameterName=SL.Pattern_Primary_BaseColor_Brightness})
  )
  n81 = NamedRerouteUsage()
  n82 = VectorParameter() {group=SL 01 Global, parameterName=SL.Secondary_BaseColorAdd, sortPriority=2, defaultValue=<rgba 0 0 0 1>}
  n83 = Multiply(A=NamedRerouteUsage(), B=ScalarParameter() {defaultValue=0.4, group=SL 01 Global, parameterName=SL.Pattern_Secondary_BaseColor_Brightness})
  n84 = LinearInterpolate(
   A=LinearInterpolate(A=LinearInterpolate(A=@n78, B=@n80, Alpha=@n81), B=Add(A=@n82.RGB, B=@n83), Alpha=NamedRerouteUsage())
   B=NamedRerouteUsage()
   Alpha=NamedRerouteUsage()
  )
  n85 = NamedRerouteUsage()
  n86 = NamedRerouteUsage()
  NamedRerouteDeclaration(None=LinearInterpolate(A=LinearInterpolate(A=@n84, B=@n85, Alpha=@n86), B=NamedRerouteUsage(), Alpha=NamedRerouteUsage())) {name=SL_BaseColor, nodeColor=<rgba 0 0.0469 1 1>}
  n87 = NamedRerouteUsage()
  n88 = NamedRerouteUsage()
  n89 = LinearInterpolate(
   A=LinearInterpolate(A=LinearInterpolate(A=@n3, B=@n87, Alpha=@n88), B=NamedRerouteUsage(), Alpha=NamedRerouteUsage())
   B=NamedRerouteUsage()
   Alpha=NamedRerouteUsage()
  )
  n90 = NamedRerouteUsage()
  n91 = NamedRerouteUsage()
  NamedRerouteDeclaration(None=LinearInterpolate(A=LinearInterpolate(A=@n89, B=@n90, Alpha=@n91), B=NamedRerouteUsage(), Alpha=NamedRerouteUsage())) {name=SL_Metallic, nodeColor=<rgba 0.5106 0.2321 1 1>}
  NamedRerouteDeclaration(None=@n8.M) {name=M_Input, nodeColor=<rgba 0.4297 0 1 1>}
  n92 = StaticSwitch(
   True=StaticSwitch(True=@n13.Pattern_Primary_Color, False=@n14.Pattern_Primary_Color, Value=@n15)
   False=@n17.Pattern_Primary_Color
   Value=@n18
  )
  n93 = LinearInterpolate(
   A=@n92
   B=Multiply(A=@n92, B=NamedRerouteUsage())
   Alpha=ScalarParameter() {defaultValue=0.5, group=SL 01 Global, parameterName=SL.Pattern_GMap_Influence}
  )
  n94 = NamedRerouteUsage()
  NamedRerouteDeclaration(None=LinearInterpolate(A=@n93, B=Power(Base=@n93), Alpha=Power(Base=@n94) {constExponent=3})) {name=PrimaryColor, nodeColor=<rgba 1 0 0 1>}
  NamedRerouteDeclaration(None=ScalarParameter() {group=SL 01 Global, parameterName=SL.Primary_Metallic, sortPriority=1}) {name=PrimaryMetallic, nodeColor=<rgba 1 0 0 1>}
  NamedRerouteDeclaration(None=ScalarParameter() {group=SL 01 Global, parameterName=SL.Primary_Specular, sortPriority=1}) {name=PrimarySpecular, nodeColor=<rgba 1 0 0 1>}
  n95 = ChannelMaskParameter(None=@n36) {group=SL 01 Global, maskChannel=Green, parameterName=SL.AccentMaskChannel, sortPriority=-27, defaultValue=<rgba 0 1 0 0>}
  n96 = NamedRerouteUsage()
  NamedRerouteDeclaration(None=Saturate(None=Subtract(A=@n95.R, B=@n96))) {name=AccentMask, nodeColor=<rgba 0.25 1 0 1>}
  NamedRerouteDeclaration(None=ScalarParameter() {group=SL 01 Global, parameterName=SL.Primary_Roughness, sortPriority=1}) {name=PrimaryRoughness, nodeColor=<rgba 1 0 0 1>}
  n97 = Desaturation(None=@n11.Base Color, Fraction=ScalarParameter() {group=SL 01 Global, parameterName=SL.Hair_Desaturation})
  n98 = OneMinus(None=NamedRerouteUsage())
  NamedRerouteDeclaration(None=StaticSwitchParameter(True=LinearInterpolate(A=@n11.Base Color, B=@n97, Alpha=@n98), False=@n11.Base Color) {group=SL 01 Global, parameterName=UseFNHair2_0}) {name=BaseColor, nodeColor=<rgba 0 1 0.25 1>}
  NamedRerouteDeclaration(None=@n11.Metallic) {name=Metallic, nodeColor=<rgba 1 0 0.0703 1>}
  NamedRerouteDeclaration(None=@n11.Specular) {name=Specular, nodeColor=<rgba 0 0.3906 1 1>}
  NamedRerouteUsage()
  NamedRerouteUsage()
  NamedRerouteUsage()
  NamedRerouteUsage()
  NamedRerouteUsage()
  VectorParameter() {group=SL 01 Global, parameterName=SL.SecondaryColor, sortPriority=1}
  VectorParameter() {group=SL 01 Global, parameterName=SL.PrimaryColor, sortPriority=1}
 ]
}
