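# Translucent glass material for the Season 41 Superlevels skins of the LocalTalent Flare female character.
# A shared glass function with masked thin film, optional camera dither and vertex hide forms the base.
# Static bools select T1, T2 or T3 pattern colors, alongside channel-masked tertiary, accent and emissive-detail layers, including Fresnel-driven thin film.
# The result adjusts emissive and opacity, applying translucency sorting before TechArt overrides.
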Material /BRCosmetics/Characters/Player/Female/Medium/Bodies/F_MED_LocalTalent_Flare/Skins/Superlevels/Materials/T1/M_SL_S41_Glass
graph {
 properties {
  allowTranslucentCustomDepthWrites=true
  bEnableResponsiveAA=true
  bScreenSpaceReflections=true
  bUseMaterialAttributes=true
  bUsedWithClothing=true
  bUsedWithMeshDeformer=true
  bUsedWithMorphTargets=true
  bUsedWithSkeletalMesh=true
  blendMode=BLEND_Translucent
  ditherOpacityMask=true
  pixelDepthOffsetMode=PDOM_Legacy
  translucencyDirectionalLightingIntensity=5
  translucencyLightingMode=TLM_SurfacePerPixelLighting
  usedShadingModels=Default Lit
 }
 nodes [
  LinearInterpolate(B=NamedRerouteUsage(), Alpha=NamedRerouteUsage())
  n1 = NamedRerouteUsage()
  n2 = NamedRerouteUsage()
  LinearInterpolate(
   A=Multiply(A=VectorParameter() {group=SL 01 Global, parameterName=SL.ExtraDetailTint, sortPriority=2, defaultValue=<rgba 1 1 1 1>}.RGB)
   B=@n1
   Alpha=@n2
  )
  LinearInterpolate(A=LinearInterpolate(B=NamedRerouteUsage(), Alpha=NamedRerouteUsage()), B=NamedRerouteUsage(), Alpha=NamedRerouteUsage())
  n3 = TextureSampleParameter2D() {
   group=00_Glass
   parameterName=TechArtMask
   samplerType=SAMPLERTYPE_Masks
   sortPriority=-96
   texture=&T_BlackMask
  }
  n4 = MaterialFunctionCall!(
   Material Attributes=MaterialFunctionCall!(TechArtMask=@n3.RGB) {materialFunction=/Game/Global/Materials/Character_Parent_2023/Glass_Parent_2023/MF_Glass}.Result
   Fresnel=Fresnel() {baseReflectFraction=0, exponent=1}
   Alpha=@n3.RGB
  ) {materialFunction=/Game/Global/MaterialFunctions/TechArt/MF_ThinFilm_TechArt}
  n5 = StaticSwitchParameter(
   True=MaterialFunctionCall!(In=@n4.Result) {materialFunction=/Game/Packages/Fortress_SharedMaterials/Base_Material_Functions/BaseCustomizationFunctions/MF_FN_Vehicle_CameraDither}.Result
   False=@n4.Result
  ) {group=00_Glass, parameterName=UseCameraDither, sortPriority=250}
  n6 = Multiply(A=Subtract(A=ObjectPositionWS(), B=WorldPosition().XYZ), B=ScalarParameter() {group=00_Glass, parameterName=VertexHide, sortPriority=301})
  n7 = GetMaterialAttributes(None=StaticSwitchParameter(
    True=SetMaterialAttributes!(MaterialAttributes=@n5, World Position Offset=@n6) {attributeSetTypes [#1]}
    False=MaterialFunctionCall!(Material=@n5) {materialFunction=/Game/Global/Materials/Character_Parent_2023/Functions/BaseFN_CharacterOpacity_2023}.Result
   ) {group=00_Glass, parameterName=UseVertexHide, sortPriority=300}) {attributeGetTypes [#2, #3, #4]}
  n8 = NamedRerouteDeclaration(None=@n7.Emissive Color) {name=Emissive, nodeColor=<rgba 1 0.4435 0.2797 1>}
  n9 = NamedRerouteUsage()
  n10 = LinearInterpolate(A=@n7.Opacity, Alpha=NamedRerouteUsage())
  MaterialFunctionCall(In=SetMaterialAttributes!(
    MaterialAttributes=@n7.MaterialAttributes
    Emissive Color=Add(A=@n8, B=@n9)
    Opacity=MaterialFunctionCall!(Opacity=@n10) {materialFunction=/Game/Global/Materials/MaterialGym/Materials/SortedOutline/MF_SortTranslucency}.Result
   ) {attributeSetTypes [#2, #3]}) {materialFunction=&MF_TechArtOverrides}
  n11 = Multiply(A=@n1, B=@n2)
  n12 = ScalarParameter() {defaultValue=1, group=SL 01 Global, parameterName=SL.Pattern_Emissive_Brightness}
  n13 = NamedRerouteUsage()
  n14 = NamedRerouteUsage()
  n15 = NamedRerouteUsage()
  n16 = LinearInterpolate(
   A=LinearInterpolate(A=Multiply(A=@n11, B=@n12), B=NamedRerouteUsage(), Alpha=NamedRerouteUsage())
   B=Multiply(A=LinearInterpolate(A=@n13, B=@n14, Alpha=@n15), B=NamedRerouteUsage())
   Alpha=NamedRerouteUsage()
  )
  NamedRerouteDeclaration(None=@n16) {name=SL_BaseColor, nodeColor=<rgba 0 0.0469 1 1>}
  n17 = NamedRerouteUsage()
  n18 = ChannelMaskParameter(None=@n17) {group=SL 01 Global, maskChannel=Blue, parameterName=SL.SecondaryMaskChannel, sortPriority=-29, defaultValue=<rgba 0 0 1 0>}
  NamedRerouteDeclaration(None=@n18.R) {name=SecondaryMask, nodeColor=<rgba 0.1719 0 1 1>}
  n19 = MaterialFunctionCall!(FXMask) {materialFunction=/BRCosmetics/Characters/Player/Female/Medium/Bodies/F_MED_LocalTalent_Flare/Skins/Superlevels/Materials/T3/MF_SL_S41_T3}
  n20 = MaterialFunctionCall!(FXMask) {materialFunction=/BRCosmetics/Characters/Player/Female/Medium/Bodies/F_MED_LocalTalent_Flare/Skins/Superlevels/Materials/T2/MF_SL_S41_T2}
  n21 = StaticBoolParameter() {group=SL 01 Global, parameterName=SL.Use_T3_Pattern, sortPriority=-40}
  n22 = MaterialFunctionCall!(FXMask) {materialFunction=&/MF_SL_S41_T1}
  n23 = StaticBoolParameter() {group=SL 01 Global, parameterName=SL.Use_T2_Pattern, sortPriority=-40}
  NamedRerouteDeclaration(None=StaticSwitch(
    True=StaticSwitch(True=@n19.Pattern_Secondary_Color, False=@n20.Pattern_Secondary_Color, Value=@n21)
    False=@n22.Pattern_Secondary_Color
    Value=@n23
   )) {name=SecondaryColor, nodeColor=<rgba 0 0 1 1>}
  n24 = VectorParameter() {group=SL 01 Global, parameterName=SL.TertiaryColor, sortPriority=2}
  n25 = ScalarParameter() {defaultValue=1, group=SL 01 Global, parameterName=SL.Tertiary_ThinFilmScale, sortPriority=5}
  n26 = Fresnel() {exponent=0.5}
  n27 = Power(Base=@n26, Exp=ScalarParameter() {defaultValue=0.5, group=SL 01 Global, parameterName=SL.Tertiary_ThinFilmWarp, sortPriority=4})
  n28 = TextureSampleParameter2D(UVs=Add(A=Multiply(A=@n25, B=@n27), B=ScalarParameter() {group=SL 01 Global, parameterName=SL.Tertiary_ThinFilmOffset})) {
   group=SL 01 Global
   parameterName=SL.Tertiary_ThinFilmTexture
   samplerSource=SSM_Wrap_WorldGroupSettings
   sortPriority=1
   texture=/Game/Global/Textures/Default/Thinfilm/T_ThinFilm_Spectrum_COLOR
  }
  n29 = ScalarParameter() {defaultValue=0.5, group=SL 01 Global, parameterName=SL.Global_ThinFilmIntensity, sortPriority=2}
  n30 = ScalarParameter() {defaultValue=1, group=SL 01 Global, parameterName=SL.Tertiary_ThinFilmScale, sortPriority=5}
  n31 = Power(Base=@n26, Exp=ScalarParameter() {defaultValue=0.5, group=SL 01 Global, parameterName=SL.Tertiary_ThinFilmWarp, sortPriority=4})
  n32 = Multiply(
   A=Multiply(A=Multiply(A=@n28.RGB, B=@n29), B=Multiply(A=@n30, B=@n31))
   B=ScalarParameter() {defaultValue=0.5, group=SL 01 Global, parameterName=SL.Tertiary_ThinFilmIntensity, sortPriority=2}
  )
  n33 = Power(Base=@n26, Exp=ScalarParameter() {defaultValue=2, group=SL 01 Global, parameterName=SL.Global_ThinFilmExponent, sortPriority=3})
  n34 = NamedRerouteUsage()
  n35 = LinearInterpolate(
   A=@n33
   B=Multiply(A=LinearInterpolate(Alpha=@n34) {constA=1, constB=0.01}, B=@n33)
   Alpha=ScalarParameter() {group=SL 01 Global, parameterName=SL.Global_RoughnessInfluence}
  )
  NamedRerouteDeclaration(None=LinearInterpolate(A=@n24.RGB, B=Add(A=@n24.RGB, B=@n32), Alpha=Saturate(None=@n35))) {name=TertiaryColor, nodeColor=<rgba 0 1 0 1>}
  NamedRerouteDeclaration(None=VectorParameter() {group=SL 01 Global, parameterName=SL.AccentColor, sortPriority=3}.RGB) {name=AccentColor, nodeColor=<rgba 0.1854 0.7843 1 1>}
  NamedRerouteDeclaration(None=@n3.RGB) {name=TechArtMask, nodeColor=<rgba 1 0.4922 0 1>}
  NamedRerouteDeclaration(None=ChannelMaskParameter(None=@n17) {group=SL 01 Global, parameterName=SL.PrimaryMaskChannel, sortPriority=-32}.R) {name=PrimaryMask, nodeColor=<rgba 0 0.8951 1 1>}
  n36 = TextureSampleParameter2D() {
   group=SL 01 Global
   parameterName=SL.FX2Mask
   samplerSource=SSM_Wrap_WorldGroupSettings
   samplerType=SAMPLERTYPE_Masks
   sortPriority=-30
   texture=&T_BlackMask
  }
  n37 = StaticSwitchParameter(True=NamedRerouteDeclaration(None=@n36.RGB) {name=FX2Mask, nodeColor=<rgba 1 0.4922 0 1>}, False=NamedRerouteUsage()) {defaultValue=true, parameterName=SL.UseFX2Mask}
  NamedRerouteDeclaration(None=ChannelMaskParameter(None=@n37) {group=SL 01 Global, parameterName=SL.EmissiveDetailMaskChannel, sortPriority=-27}.R) {name=EmissiveDetailMask, nodeColor=<rgba 1 0.6843 0.0075 1>}
  NamedRerouteDeclaration(None=@n7.Roughness) {name=Rougness, nodeColor=<rgba 0 0.1562 1 1>}
  n38 = ChannelMaskParameter(None=@n37) {group=SL 01 Global, maskChannel=Blue, parameterName=SL.AccentMaskChannel_02, sortPriority=-27, defaultValue=<rgba 0 0 1 0>}
  NamedRerouteDeclaration(None=@n38.R) {name=AccentMask_02, nodeColor=<rgba 1 0 0.4688 1>}
  NamedRerouteDeclaration(None=VectorParameter() {group=SL 01 Global, parameterName=SL.AccentColor_02, sortPriority=3}.RGB) {name=AccentColor_02, nodeColor=<rgba 0.1854 0.7843 1 1>}
  NamedRerouteDeclaration(None=VectorParameter() {group=SL 01 Global, parameterName=SL.EmissiveDetailColor, sortPriority=3}.RGB) {name=EmissiveDetailColor, nodeColor=<rgba 0.1854 0.7843 1 1>}
  NamedRerouteDeclaration(None=ScalarParameter() {group=SL 01 Global, parameterName=SL.EmissiveDetailStrength, sortPriority=3}) {name=EmissiveDetailStrength, nodeColor=<rgba 0.1854 0.7843 1 1>}
  NamedRerouteDeclaration(None=@n16) {name=SL_Emissive, nodeColor=<rgba 0 0.7422 1 1>}
  NamedRerouteDeclaration(None=MaterialFunctionCall(
    Brightness=ScalarParameter() {defaultValue=1, group=SL 01 Global, parameterName=SL.EmissiveFresnel_Brightness, sliderMin=4, sortPriority=6}
    Exponent=ScalarParameter() {defaultValue=1.5, group=SL 01 Global, parameterName=SL.EmissiveFresnel_Exponent, sortPriority=7}
    Vertex / Pixel Normal=PixelNormalWS()
   ) {materialFunction=&MF_Fresnel_FX}.Fresnel Inverted) {name=EmissiveFresnel, nodeColor=<rgba 1 0 0.2812 1>}
  NamedRerouteDeclaration(None=VectorParameter() {group=SL 01 Global, parameterName=SL.EmissiveFresnelColor, sortPriority=3}.RGB) {name=EmissiveFresnelColor, nodeColor=<rgba 0.1854 0.7843 1 1>}
  NamedRerouteDeclaration(None=VectorParameter() {group=SL 01 Global, parameterName=SL.EmissiveDetailBaseColor, sortPriority=3}.RGB) {name=EmissiveDetaiBaseColor, nodeColor=<rgba 0.1854 0.7843 1 1>}
  n39 = ChannelMaskParameter(None=@n17) {group=SL 01 Global, maskChannel=Green, parameterName=SL.TertiaryMaskChannel, sortPriority=-28, defaultValue=<rgba 0 1 0 0>}
  NamedRerouteDeclaration(None=@n39.R) {name=TertiaryMask, nodeColor=<rgba 0.1484 1 0 1>}
  n40 = StaticSwitch(
   True=StaticSwitch(True=@n19.Pattern_Primary_Color, False=@n20.Pattern_Primary_Color, Value=@n21)
   False=@n22.Pattern_Primary_Color
   Value=@n23
  )
  n41 = Fresnel() {exponent=1}
  NamedRerouteDeclaration(None=LinearInterpolate(A=Power(Base=@n40), B=@n40, Alpha=OneMinus(None=@n41))) {name=PrimaryColor, nodeColor=<rgba 1 0 0 1>}
  n42 = ChannelMaskParameter(None=@n37) {group=SL 01 Global, maskChannel=Green, parameterName=SL.AccentMaskChannel, sortPriority=-27, defaultValue=<rgba 0 1 0 0>}
  NamedRerouteDeclaration(None=@n42.R) {name=AccentMask, nodeColor=<rgba 1 0 0.4688 1>}
  NamedRerouteUsage()
 ]
}
