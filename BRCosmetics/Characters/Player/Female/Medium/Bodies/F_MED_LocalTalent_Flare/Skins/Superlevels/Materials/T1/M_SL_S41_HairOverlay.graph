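# Masked character hair overlay material for the Flare Superlevels skins, layering a recolor onto the 2023 character parent.
# MF_MessHairOverlay_2023 builds the initial attributes between the PreFX and PostFX functions, and Base Color is then overridden.
# A remapped, shifted gradient, optionally switched by a sphere-gradient parameter, blends two hair colors.
# The sphere gradient uses interpolated pre-skinned local position, and a channel-selectable FX2 texture supplies the hair mask.
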Material /BRCosmetics/Characters/Player/Female/Medium/Bodies/F_MED_LocalTalent_Flare/Skins/Superlevels/Materials/T1/M_SL_S41_HairOverlay
graph {
 properties {
  bUseMaterialAttributes=true
  bUsedWithClothing=true
  bUsedWithMeshDeformer=true
  bUsedWithMorphTargets=true
  bUsedWithSkeletalMesh=true
  blendMode=BLEND_Masked
  pixelDepthOffsetMode=PDOM_Legacy
  shadingModel=MSM_FromMaterialExpression
  usedShadingModels=Default Lit | Subsurface | Clear Coat | Subsurface Profile | Cloth
 }
 nodes [
  n1 = MaterialFunctionCall() {materialFunction=&MF_CharacterParent_2023_PreFX}
  MaterialFunctionCall(Material=NamedRerouteUsage(), M=@n1.M, Pre Skinned Local Position=@n1.M, Distance Blend=@n1.M, Fresnel=@n1.M) {materialFunction=&MF_CharacterParent_2023_PostFX}
  NamedRerouteDeclaration(None=MaterialFunctionCall!(MA=@n1.Result, PreSkinLOC=@n1.Result) {materialFunction=/Game/Global/Materials/Character_Parent_2023/Hair_Parent_2023/MaterialFunctions/MF_MessHairOverlay_2023}.Result) {name=Inital_MA, nodeColor=<rgba 0 0.8359 1 1>}
  n2 = GetMaterialAttributes(None=NamedRerouteUsage()) {attributeGetTypes [#1]}
  NamedRerouteDeclaration(None=SetMaterialAttributes!(MaterialAttributes=@n2.MaterialAttributes, Base Color=NamedRerouteUsage()) {attributeSetTypes [#1]}) {name=SL_MA, nodeColor=<rgba 0 1 0.2031 1>}
  n3 = TextureSampleParameter2D() {
   group=SL 01 Global
   parameterName=SL.FX2Mask
   samplerSource=SSM_Wrap_WorldGroupSettings
   samplerType=SAMPLERTYPE_Masks
   sortPriority=-30
   texture=&T_BlackMask
  }
  n4 = ChannelMaskParameter(None=@n3.RGB) {group=SL 01 Global, maskChannel=Blue, parameterName=SL.HairMaskChannel, sortPriority=-27, defaultValue=<rgba 0 0 1 0>}
  NamedRerouteDeclaration(None=@n4.R) {name=HairMask, nodeColor=<rgba 1 0 0.4688 1>}
  n5 = VectorParameter() {
   group=SL 01 Global
   parameterName=SL.Hair_GradientMask_Remap
   channelNames {a="", b=Smoothness, g=EndPoint, r=StartPoint}
   defaultValue=<rgba 0 0.5 0.1 1>
  }
  n6 = MaterialFunctionCall!(
   IN Gradient=StaticSwitchParameter(True=NamedRerouteUsage(), False=NamedRerouteUsage()) {group=SL 01 Global, parameterName=SL.Hair_Use_SphereGradient}
   StartPoint=@n5.StartPoint
   EndPoint=@n5.StartPoint
   Smoothness=@n5.StartPoint
  ) {materialFunction=&/MF_SL_GradientRemap}
  n7 = VectorParameter() {
   group=SL 01 Global
   parameterName=SL.Hair_GradientMask_Transition
   channelNames {a=Hardness, b=Offset, g=Range, r=Amount}
   defaultValue=<rgba 0 1 0 1>
  }
  n8 = Multiply(
   A=MaterialFunctionCall!(Mask In=@n6.OUT Gradient, Amount=@n7.Amount, Range=@n7.Amount, Start Offset=@n7.Amount, Sharpness=@n7.Amount) {materialFunction=/Game/Global/MaterialFunctions/MF_MaskShifting}.Result
   B=@n6.RangedMask
  )
  NamedRerouteDeclaration(None=LinearInterpolate(
    A=VectorParameter() {group=SL 01 Global, parameterName=SL.HairColor01}.RGB
    B=VectorParameter() {group=SL 01 Global, parameterName=SL.HairColor02}.RGB
    Alpha=Saturate(None=@n8)
   )) {name=HairColor, nodeColor=<rgba 0.1854 0.7843 1 1>}
  n9 = VertexInterpolator(VS=LocalPosition() {localOrigin=InstancePreSkinning}.XYZ)
  n10 = VectorParameter() {group=SL 01 Global, parameterName=SL.Hair_Gradient_SpherePos, channelNames {a=Radius, b="", g="", r=""}, defaultValue=<rgba 0 0 170 1>}
  NamedRerouteDeclaration(None=OneMinus(None=SphereMask(A=@n9.PS, B=@n10.RGB, Radius=@n10.RGB) {attenuationRadius=10, hardnessPercent=0})) {name=HairGradientMap, nodeColor=<rgba 0.2188 0 1 1>}
  n11 = NamedRerouteUsage()
  n12 = VectorParameter() {group=SL 01 Global, parameterName=SL.BaseColorTint, sortPriority=2, defaultValue=<rgba 1 1 1 1>}
  NamedRerouteDeclaration(None=LinearInterpolate(A=Multiply(A=@n11, B=@n12.RGB), B=NamedRerouteUsage(), Alpha=NamedRerouteUsage())) {name=SL_BaseColor, nodeColor=<rgba 0 0.0469 1 1>}
  NamedRerouteDeclaration(None=@n2.Base Color) {name=BaseColor, nodeColor=<rgba 0 1 0.25 1>}
  Reroute(None=@n1.Tech Art Mask)
 ]
}
